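; Towhook projectile for the Octopus grappling weapon, which sticks to what it hits and pulls the firing player toward it.
; On hit it rejects other Octopus-tagged actors and Scoreballer glass, otherwise attaching and tightening the rope.
; The dedicated server then launches the pawn using suggested arc velocity plus momentum, with a separate downward-shot case.
; It also carries vehicle grapple and target yank logic, water-entry kill, weapon-switch cleanup and break effects.
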(event Custom|WeaponCheck
  ;[1 1 1] Destroys projectil if weapon is switched.  This is so we can keep the rope FX looking good, but still let people weapon switch after grappling
  (if (not (Utilities|Equal(Object) (Variables|/Script/FortniteGame.FortPawn|GetCurrentWeapon (Variables|Default|GetPlayerPawn)) (Variables|Default|GetWeapon%20Actor)))
    (Actor|DestroyActor)))

(event EventBeginPlay
  (bind _playerpawn (Variables|Default|GetPlayerPawn))
  ;[1 1 1] Begin Play
  (bind _asfort_player_pawn (Utilities|Casting|CastToFortPlayerPawn (Game|GetInstigator)))
  (Variables|Default|SetPlayerPawn _asfort_player_pawn)
  (if (Vehicle|IsinVehicle _playerpawn)
    (bind _returnvalue (Vehicle|GetVehicleActor (Vehicle|GetVehicleInterface _playerpawn) :!class /Script/FortniteGame.FortVehicleInterface))
    (Collision|IgnoreActorwhenMoving (Variables|Collision|GetCapsuleComponent) _returnvalue true)))

(event Collision|EventHit
  (params (MyComp :type Primitive%20Component%20Object%20Reference) (Other :type Actor%20Object%20Reference) (OtherComp :type Primitive%20Component%20Object%20Reference) (bSelfMoved :type Boolean) (HitLocation :type Vector :default "0, 0, 0") (HitNormal :type Vector :default "0, 0, 0") (NormalImpulse :type Vector :default "0, 0, 0") (Hit :type Hit%20Result%20Structure%20%28by%20ref%29))
  (bind (_bblockinghit _binitialoverlap _time _distance _location _impactpoint _normal _impactnormal _physmat _hitactor) (Collision|BreakHitResult Hit))
  ;[1 1 1] Blocks Attachment to other Octopus
  (bind _returnvalue (GameplayTags|HasAnyMatchingGameplayTags(Message) _hitactor (Variables|Default|GetOctopus%20Tag)))
  (if _returnvalue
    (Delivery|Kill)
    (else
      ;[1 1 1] Stop event.  Sets variables and calls the travel event
      (if (Utilities|NotEqual(Object) self _hitactor)
        (if (not (Variables|/Script/FortniteGame.FortPawn|GetbIsDBNO (Variables|Default|GetPlayerPawn)))
          (CallFunction|HandleHits Hit)
          (if (and (Networking|IsDedicatedServer) (Utilities|IsValid (Game|GetInstigatorController) :!class /Script/Engine.KismetSystemLibrary))
            (Variables|Default|SetHitLocation (Variables|Default|GetHitLocation))
            (Grappling|Grapple)
            (else
              (Development|PrintString "Not the dedicated server"))))))))

(event EventDestroyed
  (bind _returnvalue (Transformation|GetActorLocation))
  ;[1 1 1] Destroyed
  (bind _playerpawn (Variables|Default|GetPlayerPawn)
    (:then
      (if (Variables|/Script/FortniteGame.FortPawn|GetbIsDBNO _playerpawn)
        _
        (elif (Variables|Default|GetHookStopped)
          (Effects|SpawnEmitteratLocation "/Game/Weapons/FORT_Crossbows/Crossbow/Effects/P_OctopusTowhook_Impact_Break.P_OctopusTowhook_Impact_Break" _returnvalue (Math|Conversions|RotationFromXVector (Variables|Default|GetImpactNormal)))
          (Audio|PlaySoundatLocation (Variables|Sound|GetRopeDestroySound) _returnvalue)
          (else
            (Audio|PlaySoundatLocation (Variables|Sound|GetYankBackSound) (Transformation|GetActorLocation (Variables|Default|GetPlayerPawn)) "0, 0, 0" 2.0 1.5)))))
    (:else)))

(event Custom|GrappleVehicle
  (bind _playerpawn (Variables|Default|GetPlayerPawn))
  (bind _returnvalue (Vehicle|GetVehicleInterface _playerpawn))
  (bind _returnvalue_2 (Class|FortVehicleInterface|GetOwnerLocation _returnvalue))
  (bind _momentum (Variables|Default|GetMomentum))
  (bind _vehiclevelocity (Variables|Default|GetVehicleVelocity))
  ;[1 1 1] Vehicle Grappling
  (HandleSticks|HandleMomentum)
  (Variables|Default|SetVehicleVelocity (Vehicle|GetVehicleVelocity _returnvalue))
  (bind _returnvalue_3 (Vehicle|GetVehicleActor _returnvalue :!class /Script/FortniteGame.FortVehicleInterface))
  (Utilities|Casting|CastToFortAthenaJackalVehicle _returnvalue_3
    (:then
      (Variables|VehicleGrappling|SetbInHoverboard true))
    (:CastFailed
      (Variables|VehicleGrappling|SetbInHoverboard false)))
  (bind _outlaunchvelocity (Game|SuggestProjectileVelocityCustomArc _returnvalue_2 (Variables|Default|GetHitLocation) 0.0 (select (Variables|VehicleGrappling|GetbInHoverboard) (Variables|VehicleGrappling|GetJackalArc) (Variables|VehicleGrappling|GetVehicleArc)))) ; Calculates how much force is needed to move actor to target location
  (if (> (Math|Vector|VectorLength (+ (+ _outlaunchvelocity _momentum) _vehiclevelocity)) (Math|Vector|VectorLength (+ _outlaunchvelocity _momentum)))
    (bind _outlaunchvelocity_2 (Game|SuggestProjectileVelocityCustomArc _returnvalue_2 (Variables|Default|GetHitLocation) 0.0 (select (Variables|VehicleGrappling|GetbInHoverboard) (Variables|VehicleGrappling|GetJackalArc) (Variables|VehicleGrappling|GetVehicleArc)))) ; 40% of the distance + 40% of your veloctiy
    (else
      (bind _outlaunchvelocity_3 (Game|SuggestProjectileVelocityCustomArc _returnvalue_2 (Variables|Default|GetHitLocation) 0.0 (select (Variables|VehicleGrappling|GetbInHoverboard) (Variables|VehicleGrappling|GetJackalArc) (Variables|VehicleGrappling|GetVehicleArc)))) ; Go 40% of the distance
    ))
  (Audio|BroadcastAbilityImpactatLocation "/Game/Athena/Sounds/Events/Cube/Cube_LaunchPlayer_Cue.Cube_LaunchPlayer_Cue" _returnvalue_3 (* (select (> (Math|Vector|VectorLength (+ (+ _outlaunchvelocity _momentum) _vehiclevelocity)) (Math|Vector|VectorLength (+ _outlaunchvelocity _momentum))) (+ _outlaunchvelocity_2 (Variables|Default|GetVehicleVelocity)) _outlaunchvelocity_3) (select (Variables|VehicleGrappling|GetbInHoverboard) (Variables|VehicleGrappling|GetJackalForceMultiplier) (Variables|VehicleGrappling|GetVehicleForceMultiplier))) 7.8e+02))

(event OnExitWaterEvent(WaterInteractionComponent)
  (params (WaterBody :type Fort%20Water%20Body%20Actor%20Object%20Reference) (WaterInteractionComponent :type Fort%20Water%20Interaction%20Component%20Object%20Reference) (bIsLastBody :type Boolean))
  (bind (_velocity_x _velocity_y) (Math|Vector|BreakVector (Variables|/Script/Engine.MovementComponent|GetVelocity_X (Variables|Movement|GetProjectileMovementComponent))))
  ;[1 0.677 0.1312] OnExitWater - Set vertical velocity to zero. Because this projectile uses a gravity scale of zero, any vertical velocity it has when it exits the water will persist and likely look unnatural. X + Y velocity will have been properly dampened by the drag of the water, so we can leave those alone.
  (if bIsLastBody
    (bind _returnvalue (Gameplay|GetAttachedtoActor))
    (Utilities|IsValid _returnvalue
      (:Is%20Not%20Valid
        (Projectile|SetVelocity (Variables|Movement|GetProjectileMovementComponent) (Math|Vector|MakeVector _velocity_x _velocity_y))))))

(event OnEnterWaterEvent(WaterInteractionComponent)
  (params (WaterBody :type Fort%20Water%20Body%20Actor%20Object%20Reference) (WaterInteractionComponent :type Fort%20Water%20Interaction%20Component%20Object%20Reference) (bIsFirstBody :type Boolean))
  ;[1 0.677 0.1312] OnEnterWater - immediately destroys projectile.  Disallows bad interactions with Towhook & water
  (Utilities|FlowControl|Sequence
    (:then_0
      (Networking|CanExecuteCosmeticEvents
        (:True
          (bind _outwatersurfaceinfo (WaterBody|GetWaterSurfaceInfoStruct WaterBody (Transformation|GetActorLocation) true false false))
          (bind _returnvalue (Effects|Components|ParticleSystem|SpawnFXAtLocation (Variables|Default|GetNS%20Impact%20Water) (Utilities|Struct|BreakWaterSurfaceInfo _outwatersurfaceinfo) (Math|Conversions|RotationFromXVector (.WaterPlaneNormal (Utilities|Struct|BreakWaterSurfaceInfo _outwatersurfaceinfo)))))
          (Audio|Parameter|SetFloatParameter(Message) _returnvalue "WaterDepth" (.WaterDepth (Utilities|Struct|BreakWaterSurfaceInfo _outwatersurfaceinfo)))
          (Audio|Parameter|SetFloatParameter(Message) _returnvalue "WaterBodyIndex" (Math|Conversions|ToFloat(Integer) (.WaterBodyIdx (Utilities|Struct|BreakWaterSurfaceInfo _outwatersurfaceinfo))))
          (Audio|PlaySoundatLocation (Variables|Sound|GetWaterSplashSound) (Utilities|Struct|BreakWaterSurfaceInfo _outwatersurfaceinfo)))))
    (:then_1
      (Utilities|FlowControl|Delay 0.2)
      (Delivery|Kill))))

(fn CalculateYankForce
  (bind (_outlocation _outrotation) (Actor|GetActorEyesViewPoint (Variables|Default|GetPlayerPawn)))
  (bind _outlaunchvelocity (Game|SuggestProjectileVelocityCustomArc (Transformation|GetActorLocation (Variables|Default|GetTarget)) (+ _outlocation (* (Math|Vector|GetForwardVector _outrotation) (Variables|Default|GetDistanceFromInstigator))))) ; Finds out how much force is needed to put the target 300 units in front of where instigator is facing
  (if (Utilities|Enum|NotEqual(Enum) (Networking|GetLocalRole (Variables|Default|GetTarget)) "ROLE_SimulatedProxy")
    (Character|LaunchCharacter (Variables|Default|GetTarget) (+ _outlaunchvelocity (Math|Vector|MakeVector (.x (Transformation|GetVelocity (Variables|Default|GetPlayerPawn))) (.y (Transformation|GetVelocity (Variables|Default|GetPlayerPawn))) (Math|Float|Absolute(Float) (.z (Transformation|GetVelocity (Variables|Default|GetPlayerPawn)))))) true true)))

(fn Grapple
  (bind _playerpawn (Variables|Default|GetPlayerPawn))
  (bind _playerpawn_2 (Variables|Default|GetPlayerPawn))
  (bind _charactermovement (Variables|/Script/Engine.Character|GetCharacterMovement _playerpawn_2))
  (bind _playerpawn_3 (Variables|Default|GetPlayerPawn))
  (bind _playerpawn_4 (Variables|Default|GetPlayerPawn))
  (Pawn|Components|CharacterMovement|ClearAccumulatedForces _charactermovement)
  (Components|Movement|StopMovementImmediately _charactermovement)
  (if (not (Variables|Default|GetAirControl))
    (Input|SetIgnoreMoveInput (Game|GetInstigatorController (Variables|Default|GetPlayerPawn)) true))
  (Variables|Default|GetWeapon%20Actor
    (:then
      _)
    (:else
      (bind _weapon_actor (CallFunction|SetWeaponActor))
      (Utilities|IsValid _weapon_actor
        (:Is%20Valid
          _))))
  (Audio|SpawnSoundAttached (Variables|Sound|GetTravelSound) (Variables|/Script/Engine.Actor|GetRootComponent (Variables|Default|GetWeapon%20Actor)))
  (HandleSticks|HandleMomentum)
  (FortMovement|CancelJumpAbility _playerpawn false)
  (if (and (AI|Components|NavMovement|IsFalling (Variables|/Script/Engine.Character|GetCharacterMovement _playerpawn)) (Math|Float|InRange(Float) (Variables|Default|GetFireAImPitch) (Variables|Default|GetDownPitchMin) (Variables|Default|GetDownPitchMax)))
    (Development|PrintString "down!")
    (Variables|Default|SetPreLaunchVelocity (Variables|/Script/Engine.MovementComponent|GetVelocity (Pawn|GetMovementComponent (Variables|Default|GetPlayerPawn))))
    (bind _outlaunchvelocity (Game|SuggestProjectileVelocityCustomArc (Transformation|GetActorLocation (Variables|Default|GetPlayerPawn)) (Variables|Default|GetHitLocation) 0.0 0.9))
    (if (Utilities|Enum|NotEqual(Enum) (Networking|GetLocalRole (Variables|Default|GetPlayerPawn)) "ROLE_SimulatedProxy")
      (Character|LaunchCharacter _playerpawn_4 (+ _outlaunchvelocity (Variables|Default|GetPreLaunchVelocity)) true false)
      (Feedback|SpawnGenericCameraLensEffect _playerpawn_4 "/Game/Weapons/FORT_Crossbows/Crossbow/Effects/BP_Hook_Camera_LensEffect.BP_Hook_Camera_LensEffect_C"))
    (else
      (Character|UnCrouch (Variables|Default|GetPlayerPawn))
      (bind _outlaunchvelocity_2 (Game|SuggestProjectileVelocityCustomArc (Transformation|GetActorLocation (Variables|Default|GetPlayerPawn)) (Math|Vector|MakeVector (.x (Variables|Default|GetHitLocation)) (.y (Variables|Default|GetHitLocation)) (+ (.z (Variables|Default|GetHitLocation)) (Variables|Default|GetAdditionalZ))))) ; Calculates how much force is needed to move actor to target location
      (if (Utilities|Enum|NotEqual(Enum) (Networking|GetLocalRole (Variables|Default|GetPlayerPawn)) "ROLE_SimulatedProxy")
        (Character|LaunchCharacter _playerpawn_3 (+ _outlaunchvelocity_2 (select (> (Math|Vector|VectorLength (+ _outlaunchvelocity_2 (Variables|Default|GetMomentum))) (Math|Vector|VectorLength _outlaunchvelocity_2)) (Variables|Default|GetMomentum) "0, 0, 0")) true true) ; Add impulse isn't resetting fall damage when changing momentum 
        (Feedback|SpawnGenericCameraLensEffect _playerpawn_3 "/Game/Weapons/FORT_Crossbows/Crossbow/Effects/BP_Hook_Camera_LensEffect.BP_Hook_Camera_LensEffect_C")
        (Game|Feedback|ClientStartCameraShake :self (Fortnite|GetFortPlayerControllerfromActor _playerpawn_3) :Shake "/Game/Athena/Items/Weapons/Abilities/HookGun/CamerShake_HookGunHitImpact.CamerShake_HookGunHitImpact_C" :Scale 1.0 :PlaySpace "CameraLocal")))))

(fn HandleHits
  (params (Hit :type Hit%20Result%20Structure))
  (bind _returnvalue (Transformation|GetActorLocation))
  (bind (_bblockinghit _binitialoverlap _time _distance _location _impactpoint _normal _impactnormal _physmat _hitactor) (Collision|BreakHitResult Hit))
  ;[1 0.4844 0.1146] Scoreballer: Don't want towhooks to be able to attach to the force field parts of the scoreballer enclouse to prevent people from cheesing it
  (Utilities|Casting|CastToBP_ScoreBallerGlass _hitactor
    (:then
      (Effects|SpawnEmitteratLocation :EmitterTemplate (Variables|Scoreballer|GetScoreballerZapFX) :Location _returnvalue :Scale "2.000000,2.000000,2.000000" :bAutoDestroy true :PoolingMethod "None" :bAutoActivateSystem true)
      (Audio|SpawnSoundatLocation (Variables|Scoreballer|GetScoreballerZapSound) _returnvalue)
      (Actor|DestroyActor))
    (:CastFailed
      ;[0.1386 1 0.03433] Attach to the hit actor
      (Game|Components|ProjectileMovement|Interpolation|SetInterpolatedComponent (Variables|Movement|GetProjectileMovementComponent) 0) ; Null the interpolation component so the static mesh doesn't get offset
      (Variables|Default|SetHookStopped true)
      (Variables|Default|SetHitLocation _impactpoint)
      (Class|BHookGunRope|TightenRope (Variables|Default|GetRope))
      (Variables|Default|SetImpactNormal _impactnormal)
      (Transformation|SetActorLocationAndRotation _location (Math|Conversions|RotationFromXVector (* _impactnormal -1.0)))
      (Game|Feedback|ClientPlayForceFeedback (Fortnite|GetFortPlayerControllerfromActor (Variables|Default|GetPlayerPawn)) "/Game/Weapons/ForceFeedback/FF_Generic_Throw.FF_Generic_Throw" "None" false false false)
      (Gameplay|ServerAttachtoActor Hit _hitactor (Variables|Default|GetOffset%20Distance%20from%20Physics%20Mesh) (Variables|Default|GetOffsetDistanceFromBone) 0.75)
      (Variables|Default|SetStruckActor _hitactor)
      (return 0))))

(fn HandleMomentum
  (bind _playerpawn (Variables|Default|GetPlayerPawn))
  (Utilities|IsValid (Variables|Default|GetStruckActor)
    (:Is%20Valid
      (Utilities|IsValid _playerpawn
        (:Is%20Valid
          (if (Vehicle|IsinVehicle _playerpawn)
            (Variables|Default|SetMomentum (Vehicle|GetVehicleVelocity (Vehicle|GetVehicleInterface _playerpawn)))
            (else
              (Variables|Default|SetMomentum (Transformation|GetVelocity (Variables|Default|GetStruckActor))))))))))

(fn HandleSticks
  (params (Object :type Object%20Reference) (Component :type Primitive%20Component%20Object%20Reference))
  (Utilities|FlowControl|Sequence
    (:then_0
      (HandleSticks|HandleStuckBuilding Object Component))
    (:then_1
      (HandleSticks|HandleStuckVehicle Object Component))
    (:then_2
      (HandleSticks|HandleStuckSupplyDrop Object Component))))

(fn HandleStuckBuilding
  (params (Object :type Object%20Reference) (Component :type Primitive%20Component%20Object%20Reference))
  (bind _asbuilding_wall (Utilities|Casting|CastToBuildingWall Object))
  (LocalVariables|SetLocalWall _asbuilding_wall)
  (bind _asstatic_mesh_component (Utilities|Casting|CastToStaticMeshComponent Component))
  (bind _returnvalue (BuildingWall|IsDoorComponent (LocalVariables|GetLocalWall) _asstatic_mesh_component))
  (if _returnvalue
    (Transformation|AttachActorToComponent Component "None" "KeepWorld" "KeepWorld" "KeepWorld" false)
    (Development|PrintString "HitBuildingWall" true true "(R=0.000000,G=0.660000,B=1.000000,A=1.000000)" 15.0)))

(fn HandleStuckProjectiles
  (params (Object :type Object%20Reference) (Component :type Primitive%20Component%20Object%20Reference))
  (Utilities|FlowControl|Sequence
    (:then_0
      (HandleSticks|HandleStuckBuilding Object Component))
    (:then_1
      (HandleSticks|HandleStuckVehicle Object Component))
    (:then_2
      (HandleSticks|HandleStuckSupplyDrop Object Component))))

(fn HandleStuckSupplyDrop
  (params (Object :type Object%20Reference) (Component :type Primitive%20Component%20Object%20Reference))
  (Utilities|FlowControl|Sequence
    (:then_0
      (Utilities|Casting|CastToAthenaSupplyDrop Object)
      (Transformation|AttachActorToComponent Component "None" "KeepWorld" "KeepWorld" "KeepWorld"))
    (:then_1
      (Utilities|Casting|CastToAthenaSupplyDropBalloon Object)
      (Transformation|AttachActorToComponent Component "None" "KeepWorld" "KeepWorld" "KeepWorld" false)))
  (Development|PrintString "Hit Supply Drop" true true "(R=0.000000,G=1.000000,B=0.087717,A=1.000000)" 1e+01))

(fn HandleStuckVehicle
  (params (Object :type Object%20Reference) (Component :type Primitive%20Component%20Object%20Reference))
  (Utilities|Casting|CastToFortAthenaVehicle Object)
  (Transformation|AttachActorToComponent Component "None" "KeepWorld" "KeepWorld" "KeepWorld" false)
  (Development|PrintString "Hit Vehicle" true true "(R=0.000000,G=0.921875,B=0.037376,A=1.000000)" 15.0))

(fn IncapacitateTarget
  (bind _target (Variables|Default|GetTarget))
  (bind _returnvalue (Pawn|GetController _target))
  (Components|Movement|StopMovementImmediately (Variables|/Script/Engine.Character|GetCharacterMovement (Variables|Default|GetTarget)))
  (Input|SetIgnoreMoveInput _returnvalue true)
  ;[1 1 1] Makes target face instigator
  (Pawn|SetControlRotation _returnvalue (Math|Rotator|FindLookatRotation (Transformation|GetActorLocation (Variables|Default|GetTarget)) (Transformation|GetActorLocation (Variables|Default|GetPlayerPawn)))))

(fn OnRep_HookStopped)

(fn OnRep_Momentum)

(fn OnRep_StruckActor)

(fn OnRep_Target)

(fn PickGrappleMontage
  (params (PitchAngle :type Float%20%28double-precision%29) (InHoverboard :type Boolean))
  (bind (_minpitch_6_c27dbdc64650f9ba5ebf09be2564561b _maxpitch_7_75538945434998adec50baa0edf1654a _sectiontoplay_12_2e20dbf74b43f4e8973e488482dca3f7) (Utilities|Struct|BreakToyAngleSelectorInfo (Utilities|Array|Get(aref) (Variables|Default|GetLaunchAngleVariations))))
  (LocalVariables|SetCleanedAngle (Math|Float|ClampAngle 0.0 -9e+01 9e+01))
  (for _array_element (Variables|Default|GetLaunchAngleVariations)
    (if (and (>= (LocalVariables|GetCleanedAngle) (Utilities|Struct|BreakToyAngleSelectorInfo _array_element)) (<= (LocalVariables|GetCleanedAngle) (.MaxPitch_7_75538945434998ADEC50BAA0EDF1654A (Utilities|Struct|BreakToyAngleSelectorInfo _array_element))))
      (return (.SectionToPlay_12_2E20DBF74B43F4E8973E488482DCA3F7 (Utilities|Struct|BreakToyAngleSelectorInfo _array_element)))))
  (return _sectiontoplay_12_2e20dbf74b43f4e8973e488482dca3f7))

(fn PickGrappleMontageSection
  (params (PitchAngle :type Float%20%28double-precision%29))
  (bind (_minpitch_6_c27dbdc64650f9ba5ebf09be2564561b _maxpitch_7_75538945434998adec50baa0edf1654a _sectiontoplay_12_2e20dbf74b43f4e8973e488482dca3f7) (Utilities|Struct|BreakToyAngleSelectorInfo (Utilities|Array|Get(aref) (Variables|Default|GetLaunchAngleVariations))))
  (LocalVariables|SetCleanedAngle (Math|Float|ClampAngle PitchAngle -9e+01 9e+01))
  (for _array_element (Variables|Default|GetLaunchAngleVariations)
    (if (and (>= (LocalVariables|GetCleanedAngle) (Utilities|Struct|BreakToyAngleSelectorInfo _array_element)) (<= (LocalVariables|GetCleanedAngle) (.MaxPitch_7_75538945434998ADEC50BAA0EDF1654A (Utilities|Struct|BreakToyAngleSelectorInfo _array_element))))
      (return (.SectionToPlay_12_2E20DBF74B43F4E8973E488482DCA3F7 (Utilities|Struct|BreakToyAngleSelectorInfo _array_element)))))
  (return _sectiontoplay_12_2e20dbf74b43f4e8973e488482dca3f7))

(fn PlayStickSounds
  (params (Location :type Vector) (HitActor :type Actor%20Object%20Reference))
  (Utilities|IsValid (Variables|Default|GetPlayerPawn)
    (:Is%20Valid
      (Utilities|IsValid HitActor
        (:Is%20Valid
          ;[1 1 1] Plays if an actor is hit
          (Utilities|FlowControl|Sequence
            (:then_0
              (if (Pawn|IsLocallyControlled (Variables|Default|GetPlayerPawn))
                (Audio|PlaySound2D (Variables|Sound|GetStuckToPlayerSound)) ; Plays on instigator
              ))
            (:then_1
              (if (Pawn|IsLocallyControlled (Utilities|Casting|CastToFortPlayerPawn HitActor))
                (Audio|PlaySound2D (Variables|Sound|GetStuckToPlayerSound)) ; Plays on target
                (else
                  _)))))
        (:Is%20Not%20Valid
          (if (Pawn|IsLocallyControlled (Variables|Default|GetPlayerPawn))
            ;[1 1 1] Plays if env is hit
            (Audio|PlaySound2D (Variables|Sound|GetStickSound))
            (else
              _)))))
    (:Is%20Not%20Valid
      _))
  (Audio|PlaySoundatLocation (Variables|Sound|GetStickSound) Location))

(fn SetWeaponActor
  (bind _playerpawn (Variables|Default|GetPlayerPawn))
  (Utilities|IsValid _playerpawn
    (:Is%20Valid
      (bind _asb_octopus_weapon_v2 (Utilities|Casting|CastToB_OctopusWeaponV2 (Weapon|GetCurrentWeapon _playerpawn))
        (:then
          (bind _output_get (Variables|Default|SetWeapon%20Actor _asb_octopus_weapon_v2)))
        (:CastFailed
          (Development|PrintString "CAST FAILED"))))
    (:Is%20Not%20Valid
      _))
  (return _output_get))

(fn ConstructionScript
  (|CallParentFunction))
